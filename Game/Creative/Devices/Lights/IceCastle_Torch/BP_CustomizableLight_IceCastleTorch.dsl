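; Creative customizable light Blueprint for the Ice Castle torch, recoloring and resizing its flame to match the color component's adjusted HSVA.
; Construction creates an IceKing campfire fire dynamic material on Torch_Fire.
; On cosmetic-capable instances, visual updates call UpdateFlame, delayed one tick outside the editor to collapse repeated calls.
; UpdateFlame toggles the particles by visibility, scales them by relative brightness, and feeds curve-derived colors into material and particle parameters.
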(event EventUpdateLightVisualsOnClient
  (Networking|CanExecuteCosmeticEvents
    (:True
      (if (not (Editor|InEditor))
        ;[1 1 1] Optimization (wait for next tick) as this may be called multiple times per tick (timelines). Don't want to use a Do-Once as it will only accept the first call, we want to only accept the last call essentially
        (Utilities|FlowControl|Delay 0.0)
        (else
          _))
      (CallFunction|UpdateFlame))))

(fn GetColorA
  (params (AdjustedColor :type F%20HSVA%20Structure))
  (bind _hsva (Class|CreativeColorComponent|GetColorbyCurveVector (Variables|Default|GetColorComponent) (Variables|Curves|GetColorACurveVector) AdjustedColor))
  (return (Class|CreativeColorComponentBlueprintLibrary|HSVAToRGBA _hsva)))

(fn GetEmbersColor
  (params (AdjustedColor :type F%20HSVA%20Structure))
  (bind _hsva (Class|CreativeColorComponent|GetColorbyCurveVector (Variables|Default|GetColorComponent) (Variables|Curves|GetEmbersCurveVector) AdjustedColor))
  (return (Math|Conversions|ToVector(LinearColor) (Class|CreativeColorComponentBlueprintLibrary|HSVAToRGBA _hsva))))

(fn GetFlame3DScale
  (params (AdjustedBrightness :type Float%20%28double-precision%29))
  (bind _brightness (Class|CreativeColorComponent|GetDefaultBrightness (Variables|Default|GetColorComponent)))
  (if (Math|Float|NearlyEqual(Float) _brightness 0.0 1e-05)
    (LocalVariables|SetScale AdjustedBrightness)
    (else
      (LocalVariables|SetScale (/ AdjustedBrightness _brightness))))
  (return (Math|Vector|MakeVector (LocalVariables|GetScale) (LocalVariables|GetScale) (LocalVariables|GetScale))))

(fn GetFlamesColor
  (params (AdjustedColor :type F%20HSVA%20Structure))
  (bind _hsva (Class|CreativeColorComponent|GetColorbyCurveVector (Variables|Default|GetColorComponent) (Variables|Curves|GetFlamesCurveVector) AdjustedColor))
  (return (Math|Conversions|ToVector(LinearColor) (Class|CreativeColorComponentBlueprintLibrary|HSVAToRGBA _hsva))))

(fn GetLightColor
  (params (AdjustedColor :type F%20HSVA%20Structure))
  (bind _hsva (Class|CreativeColorComponent|GetColorbyCurveVector (Variables|Default|GetColorComponent) (Variables|Curves|GetLightColorCurveVector) AdjustedColor))
  (return (Math|Conversions|ToVector(LinearColor) (Class|CreativeColorComponentBlueprintLibrary|HSVAToRGBA _hsva))))

(fn UpdateFlame
  (bind _torch_fire (Variables|Default|GetTorch_Fire))
  (bind (_adjustedcolor_hue_2_285c8ddb4fae69cbf132f4b895363bd4 _adjustedcolor_saturation_4_96689e7d4094d69c09c353901b909cb7 _adjustedcolor_brightness_6_0a16ad3f4c6c520b30f6959130f47e90) (Utilities|Struct|BreakFHSVA (Variables|GetAdjustedColor_Hue_2_285C8DDB4FAE69CBF132F4B895363BD4)))
  (bind _adjustedcolor (LocalVariables|GetAdjustedColor))
  (bind _output_get (LocalVariables|SetAdjustedColor (Class|CreativeColorComponent|GetAdjustedColor (Variables|Default|GetColorComponent))))
  (bind _bisvisible (CallFunction|IsColorVisible _output_get))
  (Components|Activation|SetActive (Variables|Default|GetTorch_Fire) _bisvisible false)
  (Transformation|SetRelativeScale3D (Variables|Default|GetTorch_Fire) (CallFunction|GetFlame3DScale _adjustedcolor_brightness_6_0a16ad3f4c6c520b30f6959130f47e90))
  (bind _torchfiremid (Variables|Default|GetTorchFireMID)
    (:then
      (Rendering|Material|SetVectorParameterValue _torchfiremid "Color_A" (CallFunction|GetColorA (LocalVariables|GetAdjustedColor)) :!class /Script/Engine.MaterialInstanceDynamic)
      (Rendering|Material|SetVectorParameterValue (Variables|Default|GetTorchFireMID) "Color_B" (Class|CreativeColorComponentBlueprintLibrary|HSVAToRGBA (LocalVariables|GetAdjustedColor)) :!class /Script/Engine.MaterialInstanceDynamic)
      (Effects|Components|ParticleSystem|SetVectorParameter _torch_fire "FlamesColor" (CallFunction|GetFlamesColor _adjustedcolor))
      (Effects|Components|ParticleSystem|SetVectorParameter _torch_fire "EmbersColor" (CallFunction|GetEmbersColor _adjustedcolor))
      (Effects|Components|ParticleSystem|SetVectorParameter _torch_fire "LightColor" (CallFunction|GetLightColor _adjustedcolor)))
    (:else)))

(fn ConstructionScript
  (Utilities|FlowControl|Sequence
    (:then_0
      (Networking|CanExecuteCosmeticEvents
        (:True
          (bind _returnvalue (Rendering|Material|CreateDynamicMaterialInstance (Variables|Default|GetTorch_Fire) 0 "/Game/Athena/Items/Traps/Campfire/Effects/MI_Fire_Campfire_Mesh_IceKing.MI_Fire_Campfire_Mesh_IceKing" "None" :!class /Script/Engine.PrimitiveComponent))
          (Variables|Default|SetTorchFireMID _returnvalue))))
    (:then_1
      (|CallParentFunction))))
